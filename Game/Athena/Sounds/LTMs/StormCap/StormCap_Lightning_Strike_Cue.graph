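# Sound cue for a lightning strike in the StormCap limited-time mode.
# A Random node picks between two strike waves with equal weight and feeds a Modulator at default settings, with cue volume reduced.
# Overridden attenuation uses logarithmic falloff, wide stereo spread and a low-pass filter over distance.
# A concurrency override caps simultaneous strikes and scales their volume, and the cue uses a third-person ranged weapon fire sound class.
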SoundCue /Game/Athena/Sounds/LTMs/StormCap/StormCap_Lightning_Strike_Cue
root=Modulator_0
nodes [
 {id=Modulator_0, children [Random_0]}
 {id=Random_0, children [WavePlayer_1,WavePlayer_2], properties {weights [1,1]}}
 {id=WavePlayer_1, properties {soundWaveAssetPtr=&/StormCap_Strike_01}}
 {id=WavePlayer_2, properties {soundWaveAssetPtr=&/StormCap_Strike_02}}
]
properties {
 bOverrideAttenuation=true
 bOverrideConcurrency=true
 maxDistance=4096
 soundClassObject=&Fort_WeaponRanged_Fire_3P_PSM
 volumeMultiplier=0.5
 attenuationOverrides {
  bAttenuateWithLPF=true
  distanceAlgorithm=Logarithmic
  falloffDistance=4096
  lPFFrequencyAtMax=2000
  lPFRadiusMax=4096
  lPFRadiusMin=2048
  stereoSpread=1000
  attenuationShapeExtents {x=0}
 }
 concurrencyOverrides {maxCount=3, volumeScale=0.5}
}
