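# Sound cue for a UI alert in the StormCap limited-time mode, named as the "Below" alert.
# It plays the StormCap_Decrease wave after a short fixed delay, at reduced volume, through the important in-game mission sound class.
# A concurrency override allows only one instance at a time and blocks new plays while one is active.
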SoundCue /Game/Athena/Sounds/LTMs/StormCap/StormCap_UIAlert_Below_Cue
root=Delay_0
nodes [
 {id=Delay_0, children [WavePlayer_1], properties {delayMax=0.15, delayMin=0.15}}
 {id=WavePlayer_1, properties {soundWaveAssetPtr=&/StormCap_Decrease_01}}
]
properties {
 bOverrideConcurrency=true
 maxDistance=2097152
 soundClassObject=&Fort_Mission_Ingame_Important_PSM
 volumeMultiplier=0.7
 concurrencyOverrides {maxCount=1, resolutionRule=PreventNew}
}
